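; Camera lens effect blueprint for the active Clean Sweep cloak from the elimination cloak boon.
; On BeginPlay it plays a timeline from the start.
; Each timeline update writes its value into the Alpha parameter of the shared cloak post-process mask-glow material parameter collection, animating that glow globally rather than per material instance.
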(event EventBeginPlay
  (bind (_direction _time) (Timeline|Timeline_0 PlayFromStart)
    (:Update
      (Rendering|Material|SetScalarParameterValue "/CherrySmokeBoonsGameplay/Gameplay/ElimCloak/FX3/MPC_Cloak_Generic_Post_MaskGlow.MPC_Cloak_Generic_Post_MaskGlow" "Alpha" _time :!class /Script/Engine.KismetMaterialLibrary))))

(fn ConstructionScript)
